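# Translucent, unlit, two-sided speedline material for the LethalVae pickaxe effects, set up for Niagara sprites and ribbons.
# Particle color RGB runs through the shared emissive adjustments function.
# Particle alpha scales a procedural mask: opposing U gradients shaped by Core and Edge Fade fade both ends, multiplied by a centered sphere mask with Radius and Hardness.
# It samples no textures and enables responsive AA and translucent velocity output.
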Material /Game/Weapons/FORT_Melee/Pickaxe_LethalVae/FX/M_Pickaxe_LethalVae_Speedlines
graph {
 properties {
  bEnableResponsiveAA=true
  bOutputTranslucentVelocity=true
  bUsedWithNiagaraRibbons=true
  bUsedWithNiagaraSprites=true
  blendMode=BLEND_Translucent
  pixelDepthOffsetMode=PDOM_Legacy
  shadingModel=MSM_Unlit
  twoSided=true
  usedShadingModels=Unlit
 }
 nodes [
  n1 = ParticleColor()
  MaterialFunctionCall(Emissive=@n1.RGB) {materialFunction=&MF_FX_EmissiveAdjustments}
  n2 = TextureCoordinate()
  n3 = ComponentMask(None=@n2) {r=true}
  n4 = ScalarParameter() {defaultValue=1, parameterName=Core, sliderMax=5}
  n5 = ScalarParameter() {defaultValue=1, parameterName=Edge Fade, sliderMax=5}
  n6 = OneMinus(None=@n3)
  n7 = Multiply(A=Power(Base=Multiply(A=@n3, B=@n4) {constB=3}, Exp=@n5), B=Power(Base=Multiply(A=@n6, B=@n4) {constB=3}, Exp=@n5))
  n8 = Constant() {r=0.5}
  n9 = ScalarParameter() {defaultValue=0.5, parameterName=Radius}
  n10 = ScalarParameter() {parameterName=Hardness}
  Multiply(A=@n1.A, B=Multiply(A=Multiply(A=@n7) {constB=8}, B=SphereMask(A=@n2, B=@n8, Radius=@n9, Hardness=@n10)))
 ]
}
